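# Unlit, two-sided translucent Niagara material for the KneelReedy pickaxe glow VFX, drawing an eroded glow shape with distinct inner and outer colors.
# A sine-wobbled radial gradient and vertical mask form the main mask, eroded by panning bubble noise with cell-noise distortion and optional tip bias.
# A threshold separates core from edge, driving opacity and color, with dynamic-parameter edge intensity.
# Particle color, view-angle fade and depth fade modulate the result.
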Material /BRCosmetics/Weapons/FORT_Melee/Pickaxe_KneelReedy/VFX/M_Pickaxe_KneelReedy_Glow
graph {
 properties {
  bAutomaticallySetUsageInEditor=false
  bUsedWithNiagaraMeshParticles=true
  bUsedWithNiagaraSprites=true
  blendMode=BLEND_Translucent
  shadingModel=MSM_Unlit
  twoSided=true
  usedShadingModels=Unlit
 }
 nodes [
  n1 = ScalarParameter() {defaultValue=2, parameterName=AngleFade}
  n2 = NamedRerouteUsage()
  n3 = Multiply(
   A=TextureCoordinate()
   B=ComponentMask(None=VectorParameter() {parameterName=BubbleTextureTiling, defaultValue=<rgba 1.5 1.5 0 1>}.RGB) {g=true, r=true}
  )
  n4 = ComponentMask(None=VectorParameter() {parameterName=Pan Speed, defaultValue=<rgba 0.03 0.3 0 1>}.RGB) {g=true, r=true}
  n5 = TextureSampleParameter2D(UVs=Panner() {speedX=0.1}) {
   parameterName=Displacement Texture
   samplerType=SAMPLERTYPE_Grayscale
   texture=/BRCosmetics/Weapons/FORT_Melee/Pickaxe_Cube_Ninja_Male/FX/T_CellNoise
  }
  n6 = TextureSampleParameter2D(UVs=Add(A=Panner(Coordinate=@n3, Speed=@n4) {speedX=0.03, speedY=0.15}, B=Multiply(A=@n5.R) {constB=0.05})) {
   parameterName=BubbleTexture
   samplerType=SAMPLERTYPE_LinearGrayscale
   texture=/Game/Gadgets/Assets/VinderTech_GliderChute/Glider_Thrive_Female/FX/Texture/T_CloudBubble
  }
  n7 = VectorParameter() {parameterName=BubbleTextureSmoothstep, defaultValue=<rgba 0 1 0 1>}
  n8 = OneMinus(None=LinearInterpolate(A=@n6.R, B=SmoothStep(Min=@n7.R, Max=@n7.R, Value=@n6.R), Alpha=ScalarParameter() {parameterName=BubbleTextureUseSmoothstep}) {constA=1})
  n9 = ComponentMask(None=TextureCoordinate()) {g=true}
  n10 = ScalarParameter() {defaultValue=1, parameterName=TipErosionHardness}
  n11 = Saturate(None=Add(A=Multiply(A=@n9, B=@n10), B=ScalarParameter() {parameterName=TipErosionOffset}))
  n12 = Subtract(
   A=@n2
   B=LinearInterpolate(A=@n8, B=LinearInterpolate(A=@n8, Alpha=@n11) {constB=0}, Alpha=ScalarParameter() {parameterName=UseTipErosionOffset})
  )
  n13 = LinearInterpolate(A=@n12, B=Multiply(A=@n2, B=@n12), Alpha=Subtract(A=@n2))
  n14 = SmoothStep(Value=@n13)
  n15 = Round(None=@n14)
  n16 = Constant()
  n17 = If(A=@n14, "A > B"=@n16, "A == B"=@n16, "A < B"=Constant() {r=1}) {constB=0.6}
  n18 = LinearInterpolate(
   A=@n15
   B=SmoothStep(Min=ScalarParameter() {defaultValue=0.45, parameterName=SSMin}, Max=ScalarParameter() {defaultValue=1, parameterName=SSMax}, Value=@n13) {constMax=0.7}
   Alpha=@n17
  )
  n19 = ParticleColor()
  DepthFade(
   Opacity=Multiply(
    A=MaterialFunctionCall("FadeSensitivity (optional)"=@n1) {materialFunction=&FadeBasedOnViewAngle}.Result
    B=Multiply(A=@n18, B=@n19.A)
   )
   FadeDistance=ScalarParameter() {parameterName=DepthFade}
  )
  n20 = VectorParameter() {parameterName=GlowInner, defaultValue=<rgba 0.2329 0.75 0.1523 1>}
  n21 = Multiply(
   A=VectorParameter() {parameterName=GlowOuter, defaultValue=<rgba 0.0406 2 0.015 1>}.RGB
   B=DynamicParameter() {paramNames [EdgeGlowIntensity, Param2, Param3, Param4]}.EdgeGlowIntensity
  )
  MaterialFunctionCall(Emissive=Multiply(A=LinearInterpolate(A=@n20.RGB, B=@n21, Alpha=@n17), B=@n19.RGB)) {materialFunction=&MF_FX_EmissiveAdjustments}
  n22 = Add(A=Panner() {speedY=0.1}, B=Multiply(A=@n5.R) {constB=0.1})
  Multiply(
   A=@n15
   B=LinearInterpolate(Alpha=TextureSample(UVs=@n22) {samplerType=SAMPLERTYPE_Grayscale, texture=&T_CellNoise}.R) {constA=0.36}
  )
  n23 = Multiply(
   A=TextureCoordinate()
   B=ComponentMask(None=VectorParameter() {parameterName=Radial UVs, defaultValue=<rgba 2.6 0.8 0 1>}.RGB) {g=true, r=true}
  )
  n24 = VectorParameter() {parameterName=WavyNoise, defaultValue=<rgba 4 1 0.04 1>}
  n25 = Add(A=Multiply(A=TextureCoordinate(), B=@n24.R) {constB=4}, B=Multiply(A=Time(), B=@n24.G))
  n26 = LinearInterpolate(B=Sine(None=ComponentMask(None=@n25) {g=true}), Alpha=@n24.B)
  n27 = VectorParameter() {parameterName=Radial Offset, defaultValue=<rgba 1.3 0.35 0 1>}
  n28 = SmoothStep(
   Min=ScalarParameter() {parameterName=RadialSSMin}
   Max=ScalarParameter() {defaultValue=0.6, parameterName=RadialSSMax}
   Value=MaterialFunctionCall(
    UVs=Add(A=@n23, B=@n26)
    CenterPosition=ComponentMask(None=@n27.RGB) {g=true, r=true}
    Radius=ScalarParameter() {defaultValue=0.43, parameterName=Radius}
    Density=ScalarParameter() {defaultValue=1.3, parameterName=Density}
   ) {materialFunction=&RadialGradientExponential}.RadialGradientExponential
  )
  n29 = OneMinus(None=ComponentMask(None=TextureCoordinate()) {g=true})
  n30 = ScalarParameter() {defaultValue=50, parameterName=VerticalMaskHardness}
  n31 = Saturate(None=Add(A=Multiply(A=@n29, B=@n30), B=ScalarParameter() {defaultValue=-13, parameterName=VerticalMaskOffset}))
  NamedRerouteDeclaration(None=Clamp(None=Multiply(A=@n28, B=@n31))) {name=Main Mask, nodeColor=<rgba 1 0 0.8203 1>}
 ]
}
