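# Translucent unlit two-sided Niagara sprite and mesh particle material for swirling smoke on the KneelReedy pickaxe.
# A time-driven, center-masked rotation twists panning wind-normal UVs that distort a SubUV smoke texture, with the frame set by a dynamic parameter.
# Texture alpha blends two colors, tinted by particle color.
# Dynamic FadeIn and FadeOut radii drive radial gradients, eroded by rotating swirl noise, that reveal and dissolve the opacity mask.
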Material /BRCosmetics/Weapons/FORT_Melee/Pickaxe_KneelReedy/VFX/M_Pickaxe_KneelReedy_SwirlSmoke
graph {
 properties {
  bAutomaticallySetUsageInEditor=false
  bUsedWithNiagaraMeshParticles=true
  bUsedWithNiagaraSprites=true
  blendMode=BLEND_Translucent
  shadingModel=MSM_Unlit
  twoSided=true
  usedShadingModels=Unlit
 }
 nodes [
  n1 = VectorParameter() {parameterName=ColorA, defaultValue=<rgba 0.0345 0.6094 0.0537 1>}
  n2 = VectorParameter() {parameterName=ColorB, defaultValue=<rgba 0.4261 1 0.488 1>}
  n3 = TextureObjectParameter() {parameterName=Main Texture, samplerType=SAMPLERTYPE_LinearColor, texture=/Game/Effects/Fort_Effects/Textures/Smoke/T_SmokeSwirl}
  n4 = TextureCoordinate()
  n5 = SphereMask(A=@n4, B=Constant2Vector() {g=0.5, r=0.5}) {attenuationRadius=1}
  n6 = Time()
  n7 = Rotator(Coordinate=@n4, Time=Multiply(A=Multiply(A=@n5, B=@n5), B=Multiply(A=@n6)))
  n8 = TextureSample(UVs=Panner(Coordinate=@n7) {speedY=0.1}) {samplerType=SAMPLERTYPE_Normal, texture=/Game/Effects/Fort_Effects/Textures/Tile/T_Wind_N}
  n9 = ScalarParameter() {defaultValue=0.05, parameterName=DistortionStrength}
  n10 = DynamicParameter() {defaultValue=<rgba 1 0.05 2 1>, paramNames [FadeIn,FadeOut,Frame,Param4]}
  n11 = MaterialFunctionCall!(
   Texture=@n3
   UVs=Add(A=TextureCoordinate(), B=Multiply(A=@n8.R, B=@n9) {constB=0.05})
   SubImages=VectorParameter() {parameterName=SubUVFrameCount, defaultValue=<rgba 2 2 0 1>}.RGB
   Frame=@n10.Frame
  ) {materialFunction=/Engine/Functions/Engine_MaterialFunctions03/Texturing/SubUV_Function}
  n12 = ParticleColor()
  MaterialFunctionCall(Emissive=Multiply(A=LinearInterpolate(A=@n1.RGB, B=@n2.RGB, Alpha=@n11.Alpha), B=@n12.RGB)) {materialFunction=&MF_FX_EmissiveAdjustments}
  n13 = TextureSample(UVs=@n7) {samplerType=SAMPLERTYPE_LinearColor, texture=/Game/Weapons/WeaponSkins/Textures/T_Swirly_Noise_highContrast1}
  n14 = Subtract(
   A=MaterialFunctionCall(Radius=@n10.FadeIn, Density=Constant() {r=1}, Invert Density=StaticBool()) {materialFunction=&RadialGradientExponential}.RadialGradientExponential
   B=@n13.R
  )
  n15 = MaterialFunctionCall(Radius=Subtract(A=@n10.FadeIn) {constB=0.05}, Density=Constant() {r=2}) {materialFunction=&RadialGradientExponential}
  n16 = LinearInterpolate(A=Clamp(None=Multiply(A=@n14) {constB=3}), B=@n15.RadialGradientExponential, Alpha=@n15.RadialGradientExponential)
  n17 = Subtract(A=@n11.RGB, B=OneMinus(None=Saturate(None=@n16)))
  n18 = Constant() {r=1}
  n19 = StaticBool()
  n20 = Multiply(A=Subtract(
    A=MaterialFunctionCall(Radius=@n10.FadeOut, Density=@n18, Invert Density=@n19) {materialFunction=&RadialGradientExponential}.RadialGradientExponential
    B=@n13.R
   )) {constB=3}
  n21 = MaterialFunctionCall(Radius=Subtract(A=@n10.FadeOut) {constB=0.05}, Density=Constant() {r=2}) {materialFunction=&RadialGradientExponential}
  n22 = Saturate(None=LinearInterpolate(A=Clamp(None=@n20), B=@n21.RadialGradientExponential, Alpha=@n21.RadialGradientExponential))
  Multiply(A=Multiply(A=@n11.Alpha, B=Subtract(A=@n17, B=@n22)), B=@n12.A)
  ScalarParameter()
  ScalarParameter() {defaultValue=0.7, parameterName=FadeInRadius}
  ScalarParameter() {defaultValue=0.7, parameterName=FadeOutRadius}
 ]
}
