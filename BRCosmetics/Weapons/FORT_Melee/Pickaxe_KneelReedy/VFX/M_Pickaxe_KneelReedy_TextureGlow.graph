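# Translucent, two-sided glow material for the KneelReedy pickaxe VFX, usable on Niagara sprites and mesh particles.
# The glow texture's blue channel is multiplied by particle color for a tinted term and by particle alpha for a fade term.
# The fade term is further scaled by a view-angle fade and a saturated vertical UV mask, built from inverted V with parameterized hardness and offset.
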Material /BRCosmetics/Weapons/FORT_Melee/Pickaxe_KneelReedy/VFX/M_Pickaxe_KneelReedy_TextureGlow
graph {
 properties {
  bAutomaticallySetUsageInEditor=false
  bUsedWithNiagaraMeshParticles=true
  bUsedWithNiagaraSprites=true
  blendMode=BLEND_Translucent
  twoSided=true
  usedShadingModels=Default Lit
 }
 nodes [
  n1 = TextureSample() {texture=&/T_KneelReedy_Glow}
  n2 = ParticleColor()
  Multiply(A=@n1.B, B=@n2.RGB)
  n3 = ScalarParameter() {defaultValue=1, parameterName=AngleFade}
  n4 = TextureCoordinate()
  n5 = Multiply(A=OneMinus(None=ComponentMask(None=@n4) {g=true}), B=ScalarParameter() {defaultValue=30, parameterName=VerticalMaskHardness})
  n6 = ScalarParameter() {defaultValue=-10.2, parameterName=VerticalMaskOffset}
  Multiply(
   A=Multiply(
    A=Multiply(A=@n2.A, B=@n1.B)
    B=MaterialFunctionCall("FadeSensitivity (optional)"=@n3) {materialFunction=&FadeBasedOnViewAngle}.Result
   )
   B=Saturate(None=Add(A=@n5, B=@n6))
  )
 ]
}
