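# NS_KneelReady_Idle is a Niagara system for the KneelReedy pickaxe cosmetic, named as the weapon's idle effect.
# It holds four SwordSmoke emitters, apparently smoke layers on the weapon, plus one emitter named Empty.
# SystemState is the only listed module, and no renderers, data interfaces, dynamic inputs or user variables are recorded, so the emitters' internal behavior is not shown.
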NiagaraSystem /BRCosmetics/Weapons/FORT_Melee/Pickaxe_KneelReedy/VFX/NS_KneelReady_Idle
dependencies {usedDataInterfaces [], usedDynamicInputs [], usedModules [&SystemState]}
emitters {
 Empty {rendererClasses [], renderers []}
 SwordSmoke {rendererClasses [], renderers []}
 SwordSmoke001 {rendererClasses [], renderers []}
 SwordSmoke002 {rendererClasses [], renderers []}
 SwordSmoke003 {rendererClasses [], renderers []}
}
summary {
 systemName=NS_KneelReady_Idle
 emitters [
  {emitterName=SwordSmoke, rendererClasses []}
  {emitterName=SwordSmoke001, rendererClasses []}
  {emitterName=Empty, rendererClasses []}
  {emitterName=SwordSmoke002, rendererClasses []}
  {emitterName=SwordSmoke003, rendererClasses []}
 ]
 userVariables []
}
